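# Physics asset for the Red Riding Remix female pickaxe, covering its fur bones for secondary motion.
# A long capsule on mstr_fur acts as the base body, with a small zero-length capsule on each of the four dyn_fur bones.
# Each fur bone is constrained directly to mstr_fur rather than chained, with limited swing and tighter twist.
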PhysicsAsset /BRCosmetics/Weapons/FORT_Melee/Pickaxe_Red_Riding_Remix_Female/Meshes/Red_Riding_Remix_Female_Physics
bodies {
 dyn_fur_01 [{shapeType=Capsule, length=0, radius=1, center=<4 0 0>}]
 dyn_fur_02 [{shapeType=Capsule, length=0, radius=1, center=<4 0 0>}]
 dyn_fur_03 [{shapeType=Capsule, length=0, radius=1, center=<4 0 0>}]
 dyn_fur_04 [{shapeType=Capsule, length=0, radius=1, center=<4 0 -0.75>}]
 mstr_fur [{shapeType=Capsule, length=30, radius=3, center=<0 0 -6.5673>}]
}
constraints [
 {bone1Name=dyn_fur_04, bone2Name=mstr_fur, swing1LimitDegrees=30, swing2LimitDegrees=30, twistLimitDegrees=5}
 {bone1Name=dyn_fur_03, bone2Name=mstr_fur, swing1LimitDegrees=30, swing2LimitDegrees=30, twistLimitDegrees=5}
 {bone1Name=dyn_fur_02, bone2Name=mstr_fur, swing1LimitDegrees=30, swing2LimitDegrees=30, twistLimitDegrees=5}
 {bone1Name=dyn_fur_01, bone2Name=mstr_fur, swing1LimitDegrees=30, swing2LimitDegrees=30, twistLimitDegrees=5}
]
